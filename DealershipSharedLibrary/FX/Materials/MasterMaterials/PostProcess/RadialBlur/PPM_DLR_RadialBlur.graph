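# Post-process material that draws a radial blur around a world-space point, with position, radius and amplitude switchable between material parameters and the MPC_PPM collection for Sequencer.
# The point is projected to screen to build a soft sphere mask scaled by camera distance and view direction, alongside a curve-atlas camera-distance fade.
# A custom HLSL node averages dithered scene taps along the direction to the center, with quality-capped sample counts and a constant fallback on lower feature levels.
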Material /DealershipSharedLibrary/FX/Materials/MasterMaterials/PostProcess/RadialBlur/PPM_DLR_RadialBlur
graph {
 properties {
  bAutomaticallySetUsageInEditor=false
  bUsedWithNiagaraMeshParticles=true
  blendMode=BLEND_Translucent
  materialDomain=MD_PostProcess
  shadingModel=MSM_Unlit
  usedShadingModels=Unlit
 }
 nodes [
  n1 = ComponentMask(None=CollectionParameter() {collection=/DealershipSharedLibrary/FX/Materials/MasterMaterials/PostProcess/MPC_PPM, parameterName=Position}) {b=true, g=true, r=true}
  n2 = VectorParameter() {group=Position, parameterName=Position, defaultValue=<rgba 0 0 0 1>}
  n3 = MaterialFunctionCall(World Position=NamedRerouteDeclaration(None=StaticSwitchParameter(True=@n1, False=@n2.RGB) {group="!Master Controls", parameterName=Use Sequencer MPC, sortPriority=0}) {name=Center Location World Space, nodeColor=<rgba 1 0.5156 0 1>}) {materialFunction=&TransformToClipSpace}
  ConstantBiasScale(None=@n3.Clip Space XY)
  n4 = MaterialFunctionCall(Float2=MaterialFunctionCall() {materialFunction=&ScreenResolution}.Visible Resolution) {materialFunction=&BreakOutFloat2Components}
  n5 = Divide(A=@n4.R, B=@n4.R)
  n6 = MaterialFunctionCall(Float2=MaterialFunctionCall(World Position=NamedRerouteUsage()) {materialFunction=&TransformToClipSpace}.Screen aligned 0-1 UV) {materialFunction=&BreakOutFloat2Components}
  n7 = Multiply(A=@n5, B=@n6.R)
  n8 = MaterialFunctionCall(Float2=ScreenPosition().ViewportUV) {materialFunction=&BreakOutFloat2Components}
  n9 = Multiply(A=@n5, B=@n8.R)
  n10 = CollectionParameter() {collection=&MPC_PPM, parameterName=Radius}
  n11 = ScalarParameter() {defaultValue=300, group=Radial Mask, parameterName=Radial Mask - Radius, sortPriority=0}
  n12 = Subtract(A=CameraPositionWS(), B=NamedRerouteUsage())
  n13 = Length(None=@n12) {is=class MaterialExpressionLength}
  n14 = Divide(
   A=Multiply(A=StaticSwitchParameter(True=@n10, False=@n11) {group="!Master Controls", parameterName=Use Sequencer MPC, sortPriority=0}) {constB=0.64, desc=Converting to unit}
   B=@n13
  )
  n15 = Saturate(None=Divide(A=@n13, B=ScalarParameter() {group=Radial Mask, parameterName=Radial Mask - NearScaleSizeDistance}))
  n16 = OneMinus(None=DotProduct(A=@n12, B=MaterialFunctionCall() {materialFunction=&CameraDirectionVector}.Result))
  n17 = SphereMask(
   A=Subtract(A=AppendVector(A=@n7, B=@n6.G), B=AppendVector(A=@n9, B=@n8.G))
   B=Constant()
   Radius=Multiply(A=Multiply(A=@n14, B=@n15), B=Saturate(None=@n16))
   Hardness=ScalarParameter() {defaultValue=0.3, group=Radial Mask, parameterName=Radial Mask - Hardness, sortPriority=1}
  ) {attenuationRadius=1}
  NamedRerouteDeclaration(None=LinearInterpolate(
    A=ScalarParameter() {group=Radial Mask, parameterName=Radial Mask - Min, sortPriority=1}
    B=ScalarParameter() {defaultValue=1, group=Radial Mask, parameterName=Radial Mask - Max, sortPriority=1}
    Alpha=Multiply(A=@n17, B=@n17)
   )) {name=Radial Mask, nodeColor=<rgba 0.0382 0.6328 1 1>}
  n18 = ScalarParameter() {defaultValue=2500, group=Camera Fade, parameterName=Camera Fade -Min Distance}
  n19 = Subtract(A=Distance(A=NamedRerouteUsage(), B=CameraPositionWS()), B=@n18)
  n20 = Subtract(A=ScalarParameter() {defaultValue=45000, group=Camera Fade, parameterName=Camera Fade -Max Distance}, B=@n18)
  n21 = CurveAtlasRowParameter(CurveTime=Saturate(None=Divide(A=@n19, B=@n20))) {
   atlas=/DealershipSharedLibrary/FX/Textures/Curves/Curve_DepthFade_Atlas
   curve=/DealershipSharedLibrary/FX/Textures/Curves/Curve_DepthFade_Color
   parameterName=Camera Fade Curve
  }
  NamedRerouteDeclaration(None=@n21.R) {name=Camera Fade Curve, nodeColor=<rgba 1 0.0978 0.1854 1>}
  n22 = CollectionParameter() {collection=&MPC_PPM, parameterName=Amplitude}
  n23 = ScalarParameter() {defaultValue=50, parameterName=Radial Blur - Amplitude}
  MaterialExpressionMultiply_17 = Multiply(
   A=Multiply(B=StaticSwitchParameter(True=@n22, False=@n23) {group="!Master Controls", parameterName=Use Sequencer MPC, sortPriority=0}) {constA=0.001}
   B=NamedRerouteUsage()
  )
  MaterialExpressionSceneTexture_1 = SceneTexture() {sceneTextureId=PPI_PostProcessInput0}
  MaterialExpressionScreenPosition_0 = ScreenPosition()
  n24 = Constant() {r=9}
  n25 = Constant() {r=3}
  n26 = Constant() {r=6}
  n27 = Constant() {r=11}
  MaterialExpressionReroute_1 = Reroute(None=Clamp(
    None=ScalarParameter() {defaultValue=9, group=Radial Blur, parameterName=Radial Blur - Samples}
    Max=QualitySwitch(Default=@n24, Low=@n25, High=@n24, Medium=@n26, Epic=@n27)
   ) {maxDefault=11, minDefault=2})
  MaterialExpressionScalarParameter_7 = ScalarParameter() {defaultValue=0.5, group=Radial Blur, parameterName=Raidial Blur - Dither}
  MaterialExpressionNamedRerouteUsage_19 = NamedRerouteUsage()
  n28 = Custom!(
   BlurDist=@MaterialExpressionMultiply_17
   Tex=@MaterialExpressionSceneTexture_1.Color
   ScreenUV=@MaterialExpressionScreenPosition_0.ViewportUV
   ScreenPos=@MaterialExpressionScreenPosition_0.ViewportUV
   Steps=@MaterialExpressionReroute_1
   NoiseAmount=@MaterialExpressionScalarParameter_7
   ObjectClipPos=@MaterialExpressionNamedRerouteUsage_19
  ) {
   code="const int TexIndex = 14;\r\nfloat samples[11] = {0,1,-1,2,-2,3,-3,5,-5,8,-8};\r\n\r\n// If ObjectClipPos is already a ViewportUV, just use it directly.\r\n// If it's clip-space xy in [-1,1], uncomment the conversion line.\r\nfloat2 centerUV = ObjectClipPos;\r\n// centerUV = ObjectClipPos * float2(0.5, -0.5) + 0.5;\r\n\r\nfloat2 dir = centerUV - ScreenUV;          // stay in ViewportUV\r\n\r\nint IntSteps = floor(Steps);\r\nfloat random = InterleavedGradientNoise(ScreenPos.xy, float(View.StateFrameIndexMod8));\r\nfloat jitter = lerp(1.0, random, NoiseAmount);\r\n\r\nfloat4 sum = 0;\r\nfor (int i = 0; i < IntSteps; i++)\r\n{\r\n    float2 pos = ScreenUV + dir * samples[i] * BlurDist * jitter;\r\n    pos = saturate(pos);\r\n    sum += SceneTextureLookup(ViewportUVToSceneTextureUV(pos, TexIndex), TexIndex, false);\r\n}\r\nreturn sum / IntSteps;"
   description=RadialBlur
   showCode=true
   inputs [
    {
     inputName=BlurDist
     input {expression=&self:MaterialExpressionMultiply_17}
    }
    {
     inputName=Tex
     input {
      expression=&self:MaterialExpressionSceneTexture_1
      mask=1
      maskA=1
      maskB=1
      maskG=1
      maskR=1
     }
    }
    {
     inputName=ScreenUV
     input {expression=&self:MaterialExpressionScreenPosition_0}
    }
    {
     inputName=ScreenPos
     input {
      expression=&self:MaterialExpressionScreenPosition_0
      outputIndex=1
     }
    }
    {
     inputName=Steps
     input {expression=&self:MaterialExpressionReroute_1}
    }
    {
     inputName=NoiseAmount
     input {expression=&self:MaterialExpressionScalarParameter_7}
    }
    {
     inputName=ObjectClipPos
     input {expression=&self:MaterialExpressionNamedRerouteUsage_19}
    }
   ]
  }
  n29 = Constant()
  NamedRerouteDeclaration(None=FeatureLevelSwitch(Default=@n28, ES2=@n29, ES3_1=@n29, SM4_REMOVED=@n29)) {name=Chromatic Blur Out, nodeColor=<rgba 0.2656 0 1 1>}
  NamedRerouteDeclaration(None=@n3.Buffer UVs for Scene Textures) {name=Center Location Clip Space, nodeColor=<rgba 1 0 0.5859 1>}
  NamedRerouteUsage()
  Reroute(None=MaterialFunctionCall!(Center Point World Position=NamedRerouteUsage()) {materialFunction=/Engine/Functions/Engine_MaterialFunctions02/UVs/WorldSpaceAlignedScreenCoordinates}.Offset and Scaled UVS)
 ]
}
